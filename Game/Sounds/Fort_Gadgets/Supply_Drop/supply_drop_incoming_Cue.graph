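# Sound cue for the incoming supply drop gadget sound.
# It plays a single wave, supply_drop_incoming_3, through one wave player with no randomization or mixing.
# It uses the gadget-activate attenuation and the gadgets sound class.
# A concurrency override is set to prevent new instances from playing when the limit is reached, rather than stopping existing ones.
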SoundCue /Game/Sounds/Fort_Gadgets/Supply_Drop/supply_drop_incoming_Cue
root=WavePlayer_3
nodes [{id=WavePlayer_3, properties {soundWaveAssetPtr=&/supply_drop_incoming_3}}]
properties {
 attenuationSettings=&Fort_GadgetActivate_Att
 bOverrideConcurrency=true
 maxDistance=4000
 soundClassObject=/Game/Sounds/FortSoundClassesAndModes/Fort_Gadgets_StW
 volumeMultiplier=1
 concurrencyOverrides {resolutionRule=PreventNew}
}
